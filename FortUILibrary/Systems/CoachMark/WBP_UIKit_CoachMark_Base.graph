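# Base activatable coach mark popup: header, description, optional icon, directional arrow and optional dismiss and secondary buttons.
# PreConstruct applies text, creates buttons from soft classes and positions the arrow through enum maps, matching the background material's indicator direction.
# While active, button clicks and holds are relayed as dispatchers, with clicks also firing analytics.
# Anim_Outro plays reversed as the intro, and forward on deactivation before collapsing the widget.
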WidgetBlueprint /FortUILibrary/Systems/CoachMark/WBP_UIKit_CoachMark_Base
animations [Anim_Outro {
  binding_count=1
  length_seconds=0.3
  object=&self:Anim_Outro
  sequence {bindings [Root {
     id=#1
     possessed_class=GridPanel
     tracks [
      {
       type=MovieScene2DTransformTrack
       sections [{type=MovieScene2DTransformSection, channels [Scale.Y {type=MovieSceneScriptingFloatChannel, keys [[0,1],[2,1],[4,0]]}]}]
      }
      {
       type=MovieSceneFloatTrack
       display_name=Render Opacity
       sections [{type=MovieSceneFloatSection, channels [None {type=MovieSceneScriptingFloatChannel, thinned_from=3, keys [[1,1],[6,0]]}]}]
      }
     ]
    }]}
 }]
graphs [
 BP_GetDesiredFocusTarget {nodes [
   Get Desired Focus Target()
   > Return Node!(execute, ReturnValue:Widget Object Reference=Get OtherActionButton() -> (OtherActionButton:WBP UIKit Button Regular Object Reference))
  ]}
 ButtonDismissClicked {nodes [ButtonDismissClicked() {is=title K2Node_FunctionEntry}]}
 ButtonDismissHoldCompleted {nodes [ButtonDismissHoldCompleted() {is=title K2Node_FunctionEntry}]}
 ButtonOtherActionClicked {nodes [ButtonOtherActionClicked() {is=title K2Node_FunctionEntry}]}
 ButtonOtherActionHoldCompleted {nodes [ButtonOtherActionHoldCompleted() {is=title K2Node_FunctionEntry}]}
 EventGraph {nodes [
   n1 = OnButtonDismissClicked!() -> (OutputDelegate, then, Button:Common Button Base Object Reference) {is=title K2Node_CustomEvent}
   > FireFlowRiderSelect!(
    execute
    self
    ElementName:String=Get Source String from Text!(Text:Text=Get TextDismissButton() -> (TextDismissButton:Text)) -> (ReturnValue:String) {is=title K2Node_CallFunction}
   ) -> (then)
   > Call Button Dismiss Clicked!(execute:Exec, self:WBP UIKit Coach Mark Base Object Reference) -> (then:Exec) {is=title K2Node_CallDelegate}
   n2 = OnButtonDismissHoldCompleted() {is=title K2Node_CustomEvent}
   > Call Button Dismiss Hold Completed!(execute:Exec, self:WBP UIKit Coach Mark Base Object Reference) -> (then:Exec) {is=title K2Node_CallDelegate}
   n3 = OnButtonOtherActionClicked!() -> (OutputDelegate, then, Button:Common Button Base Object Reference) {is=title K2Node_CustomEvent}
   > FireFlowRiderSelect(ElementName=Get Source String from Text(Text=Get TextOtherActionButton() -> (TextOtherActionButton:Text)))
   > Call Button Other Action Clicked!(execute:Exec, self:WBP UIKit Coach Mark Base Object Reference) -> (then:Exec) {is=title K2Node_CallDelegate}
   n4 = OnButtonOtherActionHoldComplited() {is=title K2Node_CustomEvent}
   > Call Button Other Action Hold Completed!(execute:Exec, self:WBP UIKit Coach Mark Base Object Reference) -> (then:Exec) {is=title K2Node_CallDelegate}
   Event PreConstruct()
   > SetMaxDesiredWidth!(
    execute
    self:Size Box Object Reference=Get WidgetSize() -> (WidgetSize:Size Box Object Reference)
    InMaxDesiredWidth:float=Get Widget Width() -> (Widget Width:float)
   ) -> (then)
   > n5 = Sequence!(execute) -> (then_0, then_1, then_2:Exec)
   > SetArrowPosition!(execute, self, ArrowPosition:E_UI_OutBoardPosition Enum=Get ArrowPosition() -> (ArrowPosition:E_UI_OutBoardPosition Enum=NewEnumerator3)) -> (then)
   > InitIcon!(
    execute
    self
    Direction:E_UI_DirectionLeftRightNone Enum=Get IconPosition() -> (IconPosition:E_UI_DirectionLeftRightNone Enum=NewEnumerator0)
   ) -> (then)
   SetDecorators!(
    execute=@n5.then_1
    self:Rich Text Block Object Reference=[
     Get RichText_Description() -> (RichText_Description:Common Rich Text Block Object Reference)
     Get RichText_Header() -> (RichText_Header:Common Rich Text Block Object Reference)
    ]
    InDecoratorClasses:Array of Rich Text Block Decorator Class References=Get In Decorator Classes() -> (In Decorator Classes:Array of Rich Text Block Decorator Class References)
   ) -> (then)
   > SetTextHeader!(execute, self, Text:Text=Get TextHeader() -> (TextHeader:Text)) -> (then, Empty:bool)
   > Set Justification!(
    execute
    Justification:ETextJustify Enum=Get JustificationHeader() -> (JustificationHeader:ETextJustify Enum=Left)
    self:Text Layout Widget Object Reference=Get RichText_Header()
   ) -> (then, Output_Get:ETextJustify Enum=Left)
   > SetTextDescription!(execute, self, Text:Text=Get TextDescription() -> (TextDescription:Text)) -> (then, Empty:bool)
   > Set Justification(Justification=Get JustificationDescription() -> (JustificationDescription:ETextJustify Enum=Left), self=Get RichText_Description()) -> (Output_Get=Left)
   > SetDefaultColorAndOpacity!(
    execute
    self:Rich Text Block Object Reference=Get RichText_Description()
    InColorAndOpacity:Slate Color Structure=Get DefaultDescriptoinColor() -> (DefaultDescriptoinColor:Slate Color Structure)
   ) -> (then)
   Init Buttons(execute=@n5.then_2)
   > UpdateVisibilityButtonBlock()
   Event On Activated()
   > n6 = Sequence()
   > Branch(Condition=Get ShouldShowDismissButton() -> (ShouldShowDismissButton:bool))
   > Bind Event to On Clicked(self:Common Button Base Object Reference=Get DismissButton() -> (DismissButton:WBP UIKit Button Regular Object Reference), Delegate=@n1) {is=title K2Node_AddDelegate}
   > Bind Event to On Hold Completed(self:WBP UIKit Button CTA Base Object Reference=Get DismissButton(), Delegate=@n2) {is=title K2Node_AddDelegate}
   Branch(execute=@n6.then_1, Condition=Get ShouldShowOtherActionButton() -> (ShouldShowOtherActionButton:bool))
   > Bind Event to On Clicked(self=Get OtherActionButton(), Delegate=@n3)
   > Bind Event to On Hold Completed(self=Get OtherActionButton(), Delegate=@n4)
   SetVisibility(execute=@n6.then_2, InVisibility=SelfHitTestInvisible)
   > "Play Animation with Finished event (legacy)"!(
    execute:Exec
    Widget:User Widget Object Reference=Self-Reference()
    InAnimation:Widget Animation Object Reference=Get Anim_Outro() -> (Anim_Outro:Widget Animation Object Reference)
    StartAtTime:float=0
    NumLoopsToPlay:int=1
    PlayMode:EUMGSequencePlayMode Enum=Reverse
    PlaybackSpeed:float=1
   ) -> (then:Exec, Result:UMGSequence Player Object Reference, Finished:Exec) {is=title K2Node_PlayAnimation}
   Event On Deactivated()
   > n7 = Get!(execute:Exec) -> (then:Exec, else:Exec, DismissButton:WBP UIKit Button Regular Object Reference)
   > Unbind all Events from On Hold Completed!(execute:Exec, self:WBP UIKit Button CTA Base Object Reference=@n7) -> (then:Exec) {is=title K2Node_ClearDelegate}
   > n8 = Unbind all Events from On Clicked!(execute:Exec, self:Common Button Base Object Reference=@n7) -> (then:Exec) {is=title K2Node_ClearDelegate}
   n9 = Get!(execute=[@n8.then, @n7.else]) -> (then, else, OtherActionButton:WBP UIKit Button Regular Object Reference)
   > Unbind all Events from On Hold Completed(self=@n9)
   > n10 = Unbind all Events from On Clicked(self=Get OtherActionButton())
   n11 = "Play Animation with Finished event (legacy)"(execute=[@n10.then, @n9.else], Widget=Self-Reference(), InAnimation=Get Anim_Outro(), StartAtTime=0, NumLoopsToPlay=1, PlaybackSpeed=1)
   n12 = Branch(execute=@n11.Finished, Condition=IsActivated!(self:Common Activatable Widget Object Reference) -> (ReturnValue:bool))
   StopAllAnimations(execute=@n12.else, self:User Widget Object Reference)
   > SetVisibility(InVisibility=Collapsed)
  ]}
 FireFlowRiderSelect {nodes [
   FireFlowRiderSelect!() -> (then, ElementName:String) {is=title K2Node_FunctionEntry}
   n1 = Get Source String from Text(Text=Get TextDescription())
   n2 = Get TextHeader()
   n3 = Get Source String from Text(Text=@n2)
   n4 = TextIsEmpty(InText=@n2)
   > Set AdditionalInfo(AdditionalInfo:Flow Analytics Additional Info Structure="Add String (Non-Localized)"!(
     Info:"Flow Analytics Additional Info Structure (by ref)"=Get AdditionalInfo() -> (AdditionalInfo:Flow Analytics Additional Info Structure)
     Key:String=Coachmark Title
     Value:String=Select(Option 0:String=@n3, Option 1:String=@n1, Index=@n4) -> (ReturnValue:String)
    ) -> (ReturnValue:Flow Analytics Additional Info Structure)) -> (Output_Get:Flow Analytics Additional Info Structure)
   > Set AdditionalInfo(AdditionalInfo="Add String (Non-Localized)"(Info=Get AdditionalInfo(), Key=Button Name, Value=Get ElementName() -> (ElementName:String)))
   > "Fire Event Element Interacted (Non-Localized)"!(
    execute
    FortPC:Fort Player Controller Object Reference=GetFortPlayerControllerFromActor!(Actor:Actor Object Reference=GetOwningPlayer()) -> (ReturnValue:Fort Player Controller Object Reference)
    ElementName:String=Coachmark
    AdditionalInfo:Flow Analytics Additional Info Structure=Get AdditionalInfo()
   ) -> (then)
  ]}
 Init Buttons {nodes [
   Init Buttons() {is=title K2Node_FunctionEntry}
   > ClearChildren(self=Get Buttons() -> (Buttons:Horizontal Box Object Reference))
   > n1 = Sequence!(execute) -> (then_0, then_1)
   > n2 = Branch(Condition=Get ShouldShowOtherActionButton())
   > n3 = Get!(execute) -> (then, else, OtherActionButtonClass:WBP UIKit Button Regular Soft Class Reference)
   > n4 = Create Widget!(
    execute:Exec
    Class:User Widget Class Reference=Resolve Soft Reference!(Input:WBP UIKit Button Regular Soft Class Reference=@n3) -> (Output:WBP UIKit Button Regular Class Reference)
    OwningPlayer:Player Controller Object Reference
    Style:Common Button Style Class Reference=/FortUILibrary/UIKit/Buttons/Blocks/ButtonStyle_UIKit_Empty.ButtonStyle_UIKit_Empty_C
    bApplyAlphaOnDisable:bool=False
    bLocked:bool=False
    bSelectable:bool=False
    bShouldSelectUponReceivingFocus:bool=False
    bInteractableWhenSelected:bool=False
    bToggleable:bool=False
    bTriggerClickedAfterSelection:bool=False
    bDisplayInputActionWhenNotInteractable:bool=True
    bHideInputActionWithKeyboard:bool=False
    bShouldUseFallbackDefaultInputAction:bool=True
    bRequiresHold:bool=Get RequiresHoldOther() -> (RequiresHoldOther:bool)
    HoldData:Common UIHold Data Class Reference=Get HoldDataOther() -> (HoldDataOther:Common UIHold Data Class Reference)
    ClickMethod:EButtonClickMethod Enum=DownAndUp
    InputPriority:int=0
    TriggeringInputAction:Data Table Row Handle Structure=Get Input Action Other() -> (Input Action Other:Data Table Row Handle Structure)
   ) -> (then:Exec, ReturnValue:WBP UIKit Button Regular Object Reference)
   > n5 = Set OtherActionButton(OtherActionButton:WBP UIKit Button Regular Object Reference=@n4) -> (Output_Get:WBP UIKit Button Regular Object Reference)
   > SetText!(execute, self:Fort CTAButton Object Reference=@n5, ButtonText:"Text (by ref)"=Get TextOtherActionButton()) -> (then)
   > n6 = AddChildToHorizontalBox!(execute, self:Horizontal Box Object Reference=Get Buttons(), Content:Widget Object Reference=@n5) -> (then, ReturnValue:Horizontal Box Slot Object Reference)
   > SetSize!(
    execute
    self:Horizontal Box Slot Object Reference=@n6
    InSize_Value:float=1
    InSize_SizeRule:ESlateSizeRule Enum=Get ButtonFillRule() -> (ButtonFillRule:ESlateSizeRule Enum=Automatic)
   ) -> (then)
   > Branch(Condition=Get ShouldShowDismissButton())
   > SetPadding!(
    execute
    self:Horizontal Box Slot Object Reference=@n6
    InPadding_Left:float=0
    InPadding_Top:float=0
    InPadding_Right:float=12
    InPadding_Bottom:float=0
   ) -> (then)
   RemoveFromParent(execute=@n2.else, self=Get OtherActionButton())
   n7 = Branch(execute=@n1.then_1, Condition=Get ShouldShowDismissButton())
   > n8 = Get!(execute) -> (then, else, DismissButtonClass:WBP UIKit Button Regular Soft Class Reference)
   > n9 = Create Widget(
    Class=Resolve Soft Reference(Input=@n8)
    Style=&ButtonStyle_UIKit_Empty_C
    bApplyAlphaOnDisable=False
    bLocked=False
    bSelectable=False
    bShouldSelectUponReceivingFocus=False
    bInteractableWhenSelected=False
    bToggleable=False
    bTriggerClickedAfterSelection=False
    bDisplayInputActionWhenNotInteractable=True
    bHideInputActionWithKeyboard=False
    bShouldUseFallbackDefaultInputAction=True
    bRequiresHold=Get RequiresHoldDismiss() -> (RequiresHoldDismiss:bool)
    HoldData=Get HoldDataDismiss() -> (HoldDataDismiss:Common UIHold Data Class Reference)
    ClickMethod=DownAndUp
    InputPriority=0
    TriggeringInputAction=Get Input Action Dismiss() -> (Input Action Dismiss:Data Table Row Handle Structure)
   )
   > n10 = Set DismissButton(DismissButton:WBP UIKit Button Regular Object Reference=@n9) -> (Output_Get:WBP UIKit Button Regular Object Reference)
   > SetButtonType!(execute, self:WBP UIKit Button Regular Object Reference=@n10, ButtonType:E_UI_CTAButtonType Enum=NewEnumerator1) -> (then)
   > SetText(self=@n10, ButtonText=Get TextDismissButton())
   > n11 = AddChildToHorizontalBox(self=Get Buttons(), Content=@n10)
   > SetSize(self=@n11, InSize_Value=1, InSize_SizeRule=Get ButtonFillRule() -> (ButtonFillRule=Automatic))
   > Branch(Condition=Get ShouldShowOtherActionButton())
   > SetPadding(self=@n11, InPadding_Left=12, InPadding_Top=0, InPadding_Right=0, InPadding_Bottom=0)
   RemoveFromParent(execute=@n7.else, self=Get DismissButton())
  ]}
 InitArrow {nodes [
   InitArrow!() -> (then, Position:E_UI_OutBoardPosition Enum=NewEnumerator3) {is=title K2Node_FunctionEntry}
   > Set L_GridSlot(L_GridSlot:Grid Slot Object Reference=SlotAsGridSlot(Widget=Get Image_Arrow() -> (Image_Arrow:Image Object Reference))) -> (Output_Get:Grid Slot Object Reference)
   > n1 = Branch(Condition="Not Equal (Enum)"!(
     A:E_UI_OutBoardPosition Enum=Get Position() -> (Position:E_UI_OutBoardPosition Enum=NewEnumerator3)
     B:E_UI_OutBoardPosition Enum=NewEnumerator3
    ) -> (ReturnValue:bool))
   > SetVisibility(self=Get Image_Arrow(), InVisibility=SelfHitTestInvisible)
   > Set HorizontalAlignment!(
    execute
    HorizontalAlignment:EHorizontalAlignment Enum=Find(
     TargetMap:Map of E_UI_OutBoardPosition Enums to EHorizontalAlignment Enums=Get L_HorizontalRemapping() -> (L_HorizontalRemapping:Map of E_UI_OutBoardPosition Enums to EHorizontalAlignment Enums=NewEnumerator3)
     Key:"E_UI_OutBoardPosition Enum (by ref)"=Get Position() -> (Position=NewEnumerator3)
    ) -> (Value:EHorizontalAlignment Enum=HAlign_Fill)
    self:Grid Slot Object Reference=Get L_GridSlot() -> (L_GridSlot:Grid Slot Object Reference)
   ) -> (then, Output_Get:EHorizontalAlignment Enum=HAlign_Fill)
   > Set VerticalAlignment!(
    execute
    VerticalAlignment:EVerticalAlignment Enum=Find(
     TargetMap:Map of E_UI_OutBoardPosition Enums to EVerticalAlignment Enums=Get L_VerticalRemapping() -> (L_VerticalRemapping:Map of E_UI_OutBoardPosition Enums to EVerticalAlignment Enums=NewEnumerator3)
     Key=Get Position() -> (Position=NewEnumerator3)
    ) -> (Value:EVerticalAlignment Enum=VAlign_Fill)
    self:Grid Slot Object Reference=Get L_GridSlot()
   ) -> (then, Output_Get:EVerticalAlignment Enum=VAlign_Fill)
   n2 = Get Brush!(self:Image Object Reference=Get Image_Arrow()) -> (Brush:Slate Brush Structure)
   n3 = Get Position() -> (Position=NewEnumerator3)
   n4 = Get L_NudgeMap() -> (L_NudgeMap:Map of E_UI_OutBoardPosition Enums to Vector 2D Structures=NewEnumerator3)
   > Set Nudge!(
    execute
    Nudge:Vector 2D Structure=vector2d * vector2d(
     A:Vector 2D Structure=Find(TargetMap:Map of E_UI_OutBoardPosition Enums to Vector 2D Structures=@n4, Key=@n3) -> (Value:Vector 2D Structure)
     B:Vector 2D Structure=Break Slate Brush!(SlateBrush:"Slate Brush Structure (by ref)"=@n2) -> (ImageSize:Vector 2D Structure) {is=title K2Node_BreakStruct}
    ) -> (ReturnValue:Vector 2D Structure) {is=title K2Node_PromotableOperator}
    self:Grid Slot Object Reference=Get L_GridSlot()
   ) -> (then, Output_Get:Vector 2D Structure)
   > SetRenderTransformAngle!(
    execute
    self:Widget Object Reference=Get Image_Arrow()
    Angle:float=Find(
     TargetMap:"Map of E_UI_OutBoardPosition Enums to Float (double-precision)s"=Get L_RotatiomMap() -> (L_RotatiomMap:"Map of E_UI_OutBoardPosition Enums to Float (double-precision)s"=NewEnumerator3)
     Key=Get Position() -> (Position=NewEnumerator3)
    ) -> (Value:double)
   ) -> (then)
   n5 = Get ArrowPosition() -> (ArrowPosition=NewEnumerator3)
   n6 = Get OutsideOffset() -> (OutsideOffset:double)
   n7 = float * float(A=@n6, B=-1) {is=title K2Node_PromotableOperator}
   n8 = Get L_GridSlot()
   > Set Nudge(
    Nudge=vector2d + vector2d(
     A:Vector 2D Structure=Get Nudge!(self:Grid Slot Object Reference=@n8) -> (Nudge:Vector 2D Structure)
     B:Vector 2D Structure=Select!(
      NewEnumerator3:Vector 2D Structure
      NewEnumerator0_X:double=0
      NewEnumerator0_Y:double=@n7
      NewEnumerator1_X:double=0
      NewEnumerator1_Y:double=@n7
      NewEnumerator2_X:double=0
      NewEnumerator2_Y:double=@n7
      NewEnumerator4_X:double=0
      NewEnumerator4_Y:double=@n6
      NewEnumerator5_X:double=0
      NewEnumerator5_Y:double=@n6
      NewEnumerator6_X:double=0
      NewEnumerator6_Y:double=@n6
      NewEnumerator7_X:double=@n7
      NewEnumerator7_Y:double=0
      NewEnumerator8_X:double=@n7
      NewEnumerator8_Y:double=0
      NewEnumerator9_X:double=@n7
      NewEnumerator9_Y:double=0
      NewEnumerator10_X:double=@n6
      NewEnumerator10_Y:double=0
      NewEnumerator11_X:double=@n6
      NewEnumerator11_Y:double=0
      NewEnumerator12_X:double=@n6
      NewEnumerator12_Y:double=0
      Index:E_UI_OutBoardPosition Enum=@n5
     ) -> (ReturnValue:Vector 2D Structure)
    ) -> (ReturnValue:Vector 2D Structure) {is=title K2Node_PromotableOperator}
    self=Get L_GridSlot()
   )
   n9 = Get SafeZone() -> (SafeZone:double)
   n10 = Make Margin!(Left=@n9, Right=@n9) -> (Margin)
   n11 = Get SafeZone()
   n12 = Make Margin!(Top:float=@n11, Bottom:float=@n11) -> (Margin)
   > SetPadding!(
    execute
    self:Grid Slot Object Reference=Get L_GridSlot()
    InPadding:Margin Structure=Select!(
     NewEnumerator3:Margin Structure
     NewEnumerator0:Margin Structure=@n10
     NewEnumerator1:Margin Structure=@n10
     NewEnumerator2:Margin Structure=@n10
     NewEnumerator4:Margin Structure=@n10
     NewEnumerator5:Margin Structure=@n10
     NewEnumerator6:Margin Structure=@n10
     NewEnumerator7:Margin Structure=@n12
     NewEnumerator8:Margin Structure=@n12
     NewEnumerator9:Margin Structure=@n12
     NewEnumerator10:Margin Structure=@n12
     NewEnumerator11:Margin Structure=@n12
     NewEnumerator12:Margin Structure=@n12
     Index=Get ArrowPosition() -> (ArrowPosition=NewEnumerator3)
    ) -> (ReturnValue:Margin Structure)
   ) -> (then)
   SetVisibility(execute=@n1.else, self=Get Image_Arrow(), InVisibility=Collapsed)
  ]}
 InitBackground {nodes [
   InitBackground() {is=title K2Node_FunctionEntry}
   > SetBrush(
    self=Get Image_Background() -> (Image_Background:Image Object Reference)
    InBrush=Get BackgroundBrush() -> (BackgroundBrush:Slate Brush Structure)
   )
   > n1 = GetDynamicMaterial(self=Get Image_Background())
   > Set BackgroundDynamicMaterial(BackgroundDynamicMaterial:Material Instance Dynamic Object Reference=@n1) -> (Output_Get:Material Instance Dynamic Object Reference)
   > SetScalarParameterValue(
    self=Get BackgroundDynamicMaterial() -> (BackgroundDynamicMaterial:Material Instance Dynamic Object Reference)
    ParameterName=IndicatorDirection
    Value=Select(
     NewEnumerator3:float=0
     NewEnumerator0:float=0
     NewEnumerator1:float=0
     NewEnumerator2:float=0
     NewEnumerator4:float=180
     NewEnumerator5:float=180
     NewEnumerator6:float=180
     NewEnumerator7:float=270
     NewEnumerator8:float=270
     NewEnumerator9:float=270
     NewEnumerator10:float=90
     NewEnumerator11:float=90
     NewEnumerator12:float=90
     Index=Get ArrowPosition() -> (ArrowPosition=NewEnumerator3)
    ) -> (ReturnValue:float)
   )
  ]}
 InitIcon {nodes [
   n1 = InitIcon!() -> (then, Direction:E_UI_DirectionLeftRightNone Enum=NewEnumerator0) {is=title K2Node_FunctionEntry}
   > n2 = Set IconPosition(IconPosition:E_UI_DirectionLeftRightNone Enum=@n1) -> (Output_Get:E_UI_DirectionLeftRightNone Enum=NewEnumerator0)
   > n3 = Branch(Condition="Not Equal (Enum)"(A:E_UI_DirectionLeftRightNone Enum=@n2, B:E_UI_DirectionLeftRightNone Enum=NewEnumerator1))
   > n4 = Get!(execute) -> (then, else, Image_Icon:Image Object Reference)
   n5 = Get Image_Icon() -> (Image_Icon:Image Object Reference)
   AddChildToGrid!(
    execute=[@n4.then, @n7.then]
    self:Grid Panel Object Reference=Get TopGrid() -> (TopGrid:Grid Panel Object Reference)
    Content:Widget Object Reference=@n5
    InRow:int=0
    InColumn:int="To Integer (Byte)"!(InByte:Byte=Get IconPosition() -> (IconPosition=NewEnumerator0)) -> (ReturnValue:int=0)
   ) -> (then, ReturnValue:Grid Slot Object Reference)
   > SetBrush(self=@n5, InBrush=Get Brush_Icon() -> (Brush_Icon:Slate Brush Structure))
   > SetRenderTransformAngle(self=@n5, Angle=Get Angle Icon() -> (Angle Icon:float))
   > SetGridSlotAlignment!(
    execute
    Widget:Widget Object Reference=@n5
    Horizontal:EHorizontalAlignment Enum=HAlign_Center
    Vertical:EVerticalAlignment Enum=VAlign_Center
   ) -> (then)
   > SetPadding(
    self=SlotAsGridSlot(Widget=Get Description() -> (Description:Vertical Box Object Reference))
    InPadding=Select!(
     NewEnumerator0:Margin Structure=Make Margin!(Left:float=20) -> (Margin)
     NewEnumerator1:Margin Structure
     NewEnumerator2:Margin Structure=Make Margin!(Right:float=20) -> (Margin)
     Index:E_UI_DirectionLeftRightNone Enum=Get IconPosition() -> (IconPosition=NewEnumerator0)
    ) -> (ReturnValue:Margin Structure)
   )
   n6 = Construct Object from Class(execute=@n4.else, Class=&Image) -> (ReturnValue:Image Object Reference)
   > n7 = Set Image_Icon(Image_Icon:Image Object Reference=@n6) -> (Output_Get:Image Object Reference)
   RemoveFromParent(execute=@n3.else, self=Get Image_Icon())
   > Set Image_Icon()
   > SetPadding!(
    execute
    self=SlotAsGridSlot(Widget=Get Description())
    InPadding_Left:float=0
    InPadding_Top:float=0
    InPadding_Right:float=0
    InPadding_Bottom:float=0
   ) -> (then)
  ]}
 SetArrowPosition {nodes [
   n1 = SetArrowPosition!() -> (then, ArrowPosition:E_UI_OutBoardPosition Enum=NewEnumerator3) {is=title K2Node_FunctionEntry}
   > n2 = Set ArrowPosition(ArrowPosition:E_UI_OutBoardPosition Enum=@n1) -> (Output_Get:E_UI_OutBoardPosition Enum=NewEnumerator3)
   > InitArrow!(execute, self, Position:E_UI_OutBoardPosition Enum=@n2) -> (then) {is=title K2Node_CallFunction}
   > InitBackground() {is=title K2Node_CallFunction}
  ]}
 SetText {nodes [
   n1 = SetText!() -> (then, Text:Text, TextBlock:Common Rich Text Block Object Reference) {is=title K2Node_FunctionEntry}
   > n2 = Set Text(Text=@n1, self:Rich Text Block Object Reference=@n1.TextBlock)
   n3 = TextIsEmpty(InText=@n2)
   > n4 = Branch(Condition=@n3)
   > n5 = SetVisibility(self=Get TextBlock() -> (TextBlock:Common Rich Text Block Object Reference), InVisibility=Collapsed)
   Return Node!(execute=[@n6.then, @n5.then], Empty:bool=@n3)
   n6 = SetVisibility(execute=@n4.else, self=Get TextBlock(), InVisibility=SelfHitTestInvisible)
  ]}
 SetTextDescription {nodes [
   n1 = SetTextDescription!() -> (then, Text:Text) {is=title K2Node_FunctionEntry}
   > n2 = SetText!(execute, self:Self Object Reference, Text:Text=@n1, TextBlock:Common Rich Text Block Object Reference=Get RichText_Description()) -> (then, Empty:bool) {is=title K2Node_CallFunction}
   > Return Node(Empty=@n2)
  ]}
 SetTextHeader {nodes [
   n1 = SetTextHeader!() -> (then, Text:Text) {is=title K2Node_FunctionEntry}
   > n2 = SetText(Text=@n1, TextBlock=Get RichText_Header())
   > Return Node(Empty=@n2)
  ]}
 UpdateVisibilityButtonBlock {nodes [
   UpdateVisibilityButtonBlock() {is=title K2Node_FunctionEntry}
   n1 = Get ShouldShowOtherActionButton()
   n2 = Get ShouldShowDismissButton()
   > SetVisibility(
    self=Get Buttons()
    InVisibility=Select!(Option 0:ESlateVisibility Enum=Collapsed, Option 1:ESlateVisibility Enum=SelfHitTestInvisible, Index:bool=OR Boolean(A=@n1, B=@n2)) -> (ReturnValue:ESlateVisibility Enum)
   )
  ]}
]
variables [
 Angle Icon {type=number}
 ArrowPosition {type=E_UI_OutBoardPosition, default=Top_Left}
 BackgroundBrush {
  type=SlateBrush
  default {
   bIsDynamicallyLoaded=false
   drawAs=Image
   imageType=NoImage
   mirroring=NoMirror
   resourceName=None
   resourceObject=/FortUILibrary/Systems/CoachMark/MaterialInstances/MI_UIKit_CoachMarkBackground
   tiling=NoTile
   imageSize=<32 32>
   margin=<ltrb 0 0 0 0>
   outlineSettings {
    bUseBrushTransparency=false
    roundingType=HalfHeightRadius
    width=0
    color {colorUseRule=UseColor_Specified, specifiedColor=<rgba 0 0 0 0>}
    cornerRadii=<0 0 0 1>
   }
   tintColor {colorUseRule=UseColor_Specified, specifiedColor=<rgba 1 1 1 1>}
   uVRegion {bIsValid=false, max=<0 0>, min=<0 0>}
  }
 }
 BackgroundDynamicMaterial {type=MaterialInstanceDynamic}
 Brush_Icon {
  type=SlateBrush
  default {
   bIsDynamicallyLoaded=false
   drawAs=Image
   imageType=FullColor
   mirroring=NoMirror
   resourceName=None
   resourceObject=null
   tiling=NoTile
   imageSize=<64 64>
   margin=<ltrb 0 0 0 0>
   outlineSettings {
    bUseBrushTransparency=false
    roundingType=HalfHeightRadius
    width=0
    color {colorUseRule=UseColor_Specified, specifiedColor=<rgba 0 0 0 0>}
    cornerRadii=<0 0 0 1>
   }
   tintColor {colorUseRule=UseColor_Specified, specifiedColor=<rgba 1 1 1 1>}
   uVRegion {bIsValid=false, max=<0 0>, min=<0 0>}
  }
 }
 ButtonDismissClicked {}
 ButtonDismissHoldCompleted {}
 ButtonFillRule {type=ESlateSizeRule, default=Fill}
 ButtonOtherActionClicked {}
 ButtonOtherActionHoldCompleted {}
 DefaultDescriptoinColor {type=SlateColor, default {colorUseRule=UseColor_Specified, specifiedColor=<rgba 0.4233 0.7529 1 1>}}
 DismissButton {type=WBP_UIKit_Button_Regular_C, default=null}
 DismissButtonClass {type=Class, default=/FortUILibrary/UIKit/Buttons/WBP_UIKit_Button_Regular.WBP_UIKit_Button_Regular_C}
 HoldDataDismiss {type=Class, default=null}
 HoldDataOther {type=Class, default=null}
 IconPosition {type=E_UI_DirectionLeftRightNone, default=None}
 Image_Icon {type=Image, default=null}
 In Decorator Classes {type="array<Class>", default [null,null]}
 Input Action Dismiss {type=DataTableRowHandle, default {dataTable=null, rowName=None}}
 Input Action Other {type=DataTableRowHandle, default {dataTable=null, rowName=None}}
 JustificationDescription {type=ETextJustify, default=Left}
 JustificationHeader {type=ETextJustify, default=Left}
 OtherActionButton {type=WBP_UIKit_Button_Regular_C, default=null}
 OtherActionButtonClass {type=Class, default=&WBP_UIKit_Button_Regular_C}
 OutsideOffset {type=number}
 RequiresHoldDismiss {type=boolean}
 RequiresHoldOther {type=boolean}
 SafeZone {type=number, default=44}
 ShouldShowDismissButton {type=boolean, default=true}
 ShouldShowOtherActionButton {type=boolean, default=true}
 TextDescription {
  type=string
  default="<text color=\"FFF\">redesigned</> COACH MARK.<text color=\"FFF\">  Decorator - </><img id=\"MTXSmall\"/> text text text text text text text text text text text text text text "
 }
 TextDismissButton {type=string, default=Dismiss}
 TextHeader {type=string, default=Header}
 TextOtherActionButton {type=string, default=Other action}
 Widget Width {type=number, default=592}
]
widgets {
 info {inheritedWidgetCount=0, parentClass=&CommonActivatableWidget, rootWidgetClass=&GridPanel, widgetCount=12}
 widgets [
  Root {widgetClassPath=&GridPanel, widget=Root, properties {columnFill [1], rowFill [1]}}
  WBP_CaptureForPostBufferUpdate {
   widgetClassPath=&WBP_CaptureForPostBufferUpdate
   bIsVariable=true
   parent=Root
   slot {class=GridSlot, properties {layer=-1}}
  }
  Image_Arrow {
   widgetClassPath=&Image
   bIsVariable=true
   parent=Root
   widget=Image_Arrow
   properties {brush {
     bIsDynamicallyLoaded=false
     drawAs=Image
     imageType=NoImage
     mirroring=NoMirror
     resourceName=None
     resourceObject=/FortUILibrary/Systems/CoachMark/MaterialInstances/MI_UIKit_CoachMarkArrow
     tiling=NoTile
     imageSize=<40 40>
     margin=<ltrb 0 0 0 0>
     outlineSettings {
      bUseBrushTransparency=false
      roundingType=HalfHeightRadius
      width=0
      color {colorUseRule=UseColor_Specified, specifiedColor=<rgba 0 0 0 0>}
      cornerRadii=<0 0 0 0>
     }
     tintColor {colorUseRule=UseColor_Specified, specifiedColor=<rgba 1 1 1 1>}
     uVRegion {bIsValid=false, max=<0 0>, min=<0 0>}
    }}
   slot {class=GridSlot, properties {horizontalAlignment=HAlign_Center, verticalAlignment=VAlign_Top, nudge=<-18 0>, padding=<ltrb 44 0 44 0>}}
  }
  Image_Background {
   widgetClassPath=&Image
   bIsVariable=true
   parent=Root
   widget=Image_Background
   properties {brush {
     bIsDynamicallyLoaded=false
     drawAs=Image
     imageType=NoImage
     mirroring=NoMirror
     resourceName=None
     resourceObject=null
     tiling=NoTile
     imageSize=<32 32>
     margin=<ltrb 0.5 0 0.5 0>
     outlineSettings {
      bUseBrushTransparency=false
      roundingType=HalfHeightRadius
      width=0
      color {colorUseRule=UseColor_Specified, specifiedColor=<rgba 0 0 0 0>}
      cornerRadii=<0 0 0 0>
     }
     tintColor {colorUseRule=UseColor_Specified, specifiedColor=<rgba 1 1 1 1>}
     uVRegion {bIsValid=false, max=<0 0>, min=<0 0>}
    }}
   slot {class=GridSlot}
  }
  WBP_CaptureForPostBufferUpdate_1 {
   widgetClassPath=&WBP_CaptureForPostBufferUpdate
   bIsVariable=true
   parent=Root
   widget=WBP_CaptureForPostBufferUpdate_1
   slot {class=GridSlot, properties {layer=5, rowSpan=2}}
  }
  WidgetSize {
   widgetClassPath=&SizeBox
   bIsVariable=true
   parent=Root
   widget=WidgetSize
   properties {maxDesiredWidth=592}
   slot {class=GridSlot, properties {horizontalAlignment=HAlign_Center, layer=10, verticalAlignment=VAlign_Center}}
  }
  Content {
   widgetClassPath=&GridPanel
   bIsVariable=true
   parent=WidgetSize
   widget=Content
   properties {columnFill [1], rowFill [0,0]}
   slot {class=SizeBoxSlot, properties {verticalAlignment=VAlign_Center, padding=<ltrb 48 40 48 40>}}
  }
  TopGrid {
   widgetClassPath=&GridPanel
   bIsVariable=true
   parent=Content
   widget=TopGrid
   properties {columnFill [0,1,0]}
   slot {class=GridSlot}
  }
  Description {
   widgetClassPath=&VerticalBox
   bIsVariable=true
   parent=TopGrid
   widget=Description
   slot {class=GridSlot, properties {column=1, verticalAlignment=VAlign_Center}}
  }
  RichText_Header {
   widgetClassPath=&CommonRichTextBlock
   bIsVariable=true
   parent=Description
   widget=RichText_Header
   properties {
    autoWrapText=true
    bOverrideDefaultStyle=true
    justification=Center
    mobileTextBlockScale=1.1
    text="Welcome to the new sidebar!"
    defaultTextStyleOverride {
     overflowPolicy=Clip
     transformPolicy=None
     colorAndOpacity {colorUseRule=UseColor_Specified, specifiedColor=<rgba 1 1 1 1>}
     font {
      bForceMonospaced=false
      bMaterialIsStencil=false
      fontMaterial=null
      fontObject=&HeadingNow
      letterSpacing=0
      monospacedWidth=1
      size=22.5
      skewAmount=0
      typefaceFontName=75 Medium
      outlineSettings {
       bApplyOutlineToDropShadows=false
       bMiteredCorners=false
       bSeparateFillAlpha=false
       outlineMaterial=null
       outlineSize=0
       outlineColor=<rgba 0 0 0 1>
      }
     }
     highlightColor {colorUseRule=UseColor_Specified, specifiedColor=<rgba 0 0 0 1>}
     highlightShape {
      bIsDynamicallyLoaded=false
      drawAs=Image
      imageType=NoImage
      mirroring=NoMirror
      resourceName=None
      resourceObject=null
      tiling=NoTile
      imageSize=<32 32>
      margin=<ltrb 0 0 0 0>
      outlineSettings {
       bUseBrushTransparency=false
       roundingType=HalfHeightRadius
       width=0
       color {colorUseRule=UseColor_Specified, specifiedColor=<rgba 0 0 0 0>}
       cornerRadii=<0 0 0 0>
      }
      tintColor {colorUseRule=UseColor_Specified, specifiedColor=<rgba 1 1 1 1>}
      uVRegion {bIsValid=false, max=<0 0>, min=<0 0>}
     }
     selectedBackgroundColor {colorUseRule=UseColor_Foreground, specifiedColor=<rgba 1 0 1 1>}
     shadowColorAndOpacity=<rgba 0 0 0 1>
     shadowOffset=<0 0>
     strikeBrush {
      bIsDynamicallyLoaded=false
      drawAs=Image
      imageType=NoImage
      mirroring=NoMirror
      resourceName=None
      resourceObject=null
      tiling=NoTile
      imageSize=<32 32>
      margin=<ltrb 0 0 0 0>
      outlineSettings {
       bUseBrushTransparency=false
       roundingType=HalfHeightRadius
       width=0
       color {colorUseRule=UseColor_Specified, specifiedColor=<rgba 0 0 0 0>}
       cornerRadii=<0 0 0 0>
      }
      tintColor {colorUseRule=UseColor_Specified, specifiedColor=<rgba 1 1 1 1>}
      uVRegion {bIsValid=false, max=<0 0>, min=<0 0>}
     }
     underlineBrush {
      bIsDynamicallyLoaded=false
      drawAs=Image
      imageType=NoImage
      mirroring=NoMirror
      resourceName=None
      resourceObject=null
      tiling=NoTile
      imageSize=<32 32>
      margin=<ltrb 0 0 0 0>
      outlineSettings {
       bUseBrushTransparency=false
       roundingType=HalfHeightRadius
       width=0
       color {colorUseRule=UseColor_Specified, specifiedColor=<rgba 0 0 0 0>}
       cornerRadii=<0 0 0 0>
      }
      tintColor {colorUseRule=UseColor_Specified, specifiedColor=<rgba 1 1 1 1>}
      uVRegion {bIsValid=false, max=<0 0>, min=<0 0>}
     }
    }
   }
   slot {class=VerticalBoxSlot, properties {padding=<ltrb 0 0 0 16>}}
  }
  RichText_Description {
   widgetClassPath=&CommonRichTextBlock
   bIsVariable=true
   parent=Description
   widget=RichText_Description
   properties {
    autoWrapText=true
    bOverrideDefaultStyle=true
    justification=Center
    mobileTextBlockScale=1.1
    text="Welcome to the new sidebar!"
    defaultTextStyleOverride {
     overflowPolicy=Clip
     transformPolicy=None
     colorAndOpacity {colorUseRule=UseColor_Specified, specifiedColor=<rgba 0.4233 0.7529 1 1>}
     font {
      bForceMonospaced=false
      bMaterialIsStencil=false
      fontMaterial=null
      fontObject=&HeadingNow
      letterSpacing=0
      monospacedWidth=1
      size=15.75
      skewAmount=0
      typefaceFontName=74 Regular
      outlineSettings {
       bApplyOutlineToDropShadows=false
       bMiteredCorners=false
       bSeparateFillAlpha=false
       outlineMaterial=null
       outlineSize=0
       outlineColor=<rgba 0 0 0 1>
      }
     }
     highlightColor {colorUseRule=UseColor_Specified, specifiedColor=<rgba 0 0 0 1>}
     highlightShape {
      bIsDynamicallyLoaded=false
      drawAs=Image
      imageType=NoImage
      mirroring=NoMirror
      resourceName=None
      resourceObject=null
      tiling=NoTile
      imageSize=<32 32>
      margin=<ltrb 0 0 0 0>
      outlineSettings {
       bUseBrushTransparency=false
       roundingType=HalfHeightRadius
       width=0
       color {colorUseRule=UseColor_Specified, specifiedColor=<rgba 0 0 0 0>}
       cornerRadii=<0 0 0 0>
      }
      tintColor {colorUseRule=UseColor_Specified, specifiedColor=<rgba 1 1 1 1>}
      uVRegion {bIsValid=false, max=<0 0>, min=<0 0>}
     }
     selectedBackgroundColor {colorUseRule=UseColor_Foreground, specifiedColor=<rgba 1 0 1 1>}
     shadowColorAndOpacity=<rgba 0 0 0 1>
     shadowOffset=<0 0>
     strikeBrush {
      bIsDynamicallyLoaded=false
      drawAs=Image
      imageType=NoImage
      mirroring=NoMirror
      resourceName=None
      resourceObject=null
      tiling=NoTile
      imageSize=<32 32>
      margin=<ltrb 0 0 0 0>
      outlineSettings {
       bUseBrushTransparency=false
       roundingType=HalfHeightRadius
       width=0
       color {colorUseRule=UseColor_Specified, specifiedColor=<rgba 0 0 0 0>}
       cornerRadii=<0 0 0 0>
      }
      tintColor {colorUseRule=UseColor_Specified, specifiedColor=<rgba 1 1 1 1>}
      uVRegion {bIsValid=false, max=<0 0>, min=<0 0>}
     }
     underlineBrush {
      bIsDynamicallyLoaded=false
      drawAs=Image
      imageType=NoImage
      mirroring=NoMirror
      resourceName=None
      resourceObject=null
      tiling=NoTile
      imageSize=<32 32>
      margin=<ltrb 0 0 0 0>
      outlineSettings {
       bUseBrushTransparency=false
       roundingType=HalfHeightRadius
       width=0
       color {colorUseRule=UseColor_Specified, specifiedColor=<rgba 0 0 0 0>}
       cornerRadii=<0 0 0 0>
      }
      tintColor {colorUseRule=UseColor_Specified, specifiedColor=<rgba 1 1 1 1>}
      uVRegion {bIsValid=false, max=<0 0>, min=<0 0>}
     }
    }
   }
   slot {class=VerticalBoxSlot}
  }
  Buttons {
   widgetClassPath=&HorizontalBox
   bIsVariable=true
   parent=Content
   widget=Buttons
   slot {class=GridSlot, properties {horizontalAlignment=HAlign_Left, row=1, verticalAlignment=VAlign_Bottom, padding=<ltrb 0 32 0 0>}}
  }
 ]
}
